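# Translucent, two-sided material for the Sog cosmetic vehicle wheel, rendering a glowing clover rim effect on skeletal meshes.
# The mask texture's blue channel, multiplied by a time-based cosine when Pulse is on or by itself when off, is scaled by intensity and color.
# That result drives both Emissive and Opacity, then passes through shared vehicle damage and de-rez transition functions.
# Use_BlueChannel has no effect, as both branches read blue.
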Material /VehicleCosmetics/Wheels/Sog/Materials/M_Wheel_Sog
graph {
 properties {
  bAutomaticallySetUsageInEditor=false
  bUseMaterialAttributes=true
  bUsedWithSkeletalMesh=true
  blendMode=BLEND_Translucent
  twoSided=true
  usedShadingModels=Default Lit
 }
 nodes [
  n1 = TextureSampleParameter2D(UVs=TextureCoordinate()) {parameterName=Mask, samplerType=SAMPLERTYPE_LinearColor, texture=/VehicleCosmetics/Wheels/_SharedTextures/FX/cloverFXrim_RGB}
  n2 = StaticSwitchParameter(True=@n1.B, False=@n1.B) {defaultValue=true, parameterName=Use_BlueChannel}
  n3 = StaticSwitchParameter(True=Cosine(None=Time() {bIgnorePause=true}) {period=6}, False=@n2) {defaultValue=true, parameterName=Pulse}
  n4 = Multiply(
   A=Multiply(A=Multiply(A=@n3, B=@n2), B=ScalarParameter() {defaultValue=1, parameterName=Emissive_Intenisty})
   B=VectorParameter() {parameterName=Emissive_Color, defaultValue=<rgba 1 1 1 1>}.RGB
  )
  MaterialFunctionCall(In=MaterialFunctionCall(Material=SetMaterialAttributes!(MaterialAttributes, Emissive Color=@n4, Opacity=@n4) {attributeSetTypes [#1, #2]}) {materialFunction=&MF_BaseVehicle_Damage}.Result) {materialFunction=&MF_VehicleDeRez_Transition}
 ]
}
